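; Anim notify that triggers a ToyFox melee attack trace at a chosen point in a swing animation.
; On notify, it gets the mesh component's owner, finds its B_ToyFoxWeaponDestructionComponent and calls PerformAttackTrace with the notify's SwingIndex variable.
; An invalid mesh, owner or component is written to the shipping log instead.
; The notify displays as "PerformMeleeAttackTrace".
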(fn GetNotifyName
  (return "PerformMeleeAttackTrace"))

(fn Received_Notify
  (params (MeshComp :type Skeletal%20Mesh%20Component%20Object%20Reference) (Animation :type Anim%20Sequence%20Base%20Object%20Reference) (EventReference :type Anim%20Notify%20Event%20Reference%20Structure%20%28by%20ref%29))
  ;[0.01886 0.15 0.1447 0.5] ReceivedNotify: Let the weapon destruction component know to do the attack trace.
  (Utilities|IsValid MeshComp
    (:Is%20Valid
      (bind _returnvalue (Components|GetOwner MeshComp))
      (Utilities|IsValid _returnvalue
        (:Is%20Valid
          (bind _returnvalue_2 (Actor|GetComponentbyClass _returnvalue "/ToyFoxGameplay/Gameplay/Weapon/B_ToyFoxWeaponDestructionComponent.B_ToyFoxWeaponDestructionComponent_C"))
          (Utilities|IsValid _returnvalue_2
            (:Is%20Valid
              (Class|BToyFoxWeaponDestructionComponent|PerformAttackTrace _returnvalue_2 (Variables|Default|GetSwingIndex))
              (return true))
            (:Is%20Not%20Valid
              (Fortnite|FortShippingLog self "ReceivedNotify: WeaponDestructionComponent is invalid." true))))
        (:Is%20Not%20Valid
          (Fortnite|FortShippingLog self "ReceivedNotify: Owner is invalid." true))))
    (:Is%20Not%20Valid
      (Fortnite|FortShippingLog self "ReceivedNotify: MeshComponent is invalid." true))))
