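; ToyFox melee gameplay ability that plays the attack montage, emits AI noise and tracks swings with gameplay effects.
; Weapon destruction component targeting results drive boss damage, projectile explosions and server-only building damage.
; Building hits feed one batched async damage action, with work deferred across ticks for performance.
; A single destruction cue per targeting pass receives the hit buildings' volume-weighted center and computed extent, with swing index as magnitude.
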(event Ability|EventActivateAbility
  (bind _returnvalue (Fortnite|Athena|GameState|GetGameStateAthena))
  ;[0 0.15 0.01188 0.5] ActivateAbility: Cache necessary variables and play the melee attack montage.
  (|CallParentFunction)
  (Utilities|FlowControl|Sequence
    (:then_0
      (bind _returnvalue_2 (Ability|ApplyGameplayEffectToOwner (Variables|Settings|GetMeleeActiveGameplayEffectClass)))
      (Variables|Transient|SetMeleeActiveGameplayEffectHandle _returnvalue_2)
      ;;[0.02747 0.1394 0.15 0.5] Cache variables and bind delegates.
      (bind _outfortplayerpawn (CallFunction|GetAvatarasValidFortPlayerPawn)
        (:ValidFortPlayerPawn
          (bind _output_get (Variables|Transient|SetAvatarPlayerPawn _outfortplayerpawn))
          (bind _output_get_2 (Variables|Transient|SetWeaponDestructionComponent (Actor|GetComponentbyClass _output_get "/ToyFoxGameplay/Gameplay/Weapon/B_ToyFoxWeaponDestructionComponent.B_ToyFoxWeaponDestructionComponent_C")))
          (Utilities|IsValid _output_get_2
            (:Is%20Valid
              (Default|BindEventtoOnTargetingExecuted _output_get_2 (EventDispatchers|CreateEvent :!event Custom|HandleTargetingExecuted))
              (Utilities|IsValid _returnvalue
                (:Is%20Valid
                  (Variables|Default|SetGiantPawnDestructionManager (Actor|GetComponentbyClass _returnvalue "/GiantPawnCore/Gameplay/Destruction/B_GiantPawnDestructionManagerGameStateComponent.B_GiantPawnDestructionManagerGameStateComponent_C")))))))))
    (:then_1
      ;;[0.03689 0.04106 0.15 0.5] Add noise event for AI to react to
      (bind _outfortplayerpawn_2 (CallFunction|GetAvatarasValidFortPlayerPawn)
        (:ValidFortPlayerPawn
          (AI|Perception|MakeNoiseEventatLocation _outfortplayerpawn_2 (ScalableFloat|GetValueatLevel (Variables|Hotfix|GetHotfix_PerceptionRange)) (Transformation|GetActorLocation _outfortplayerpawn_2) "WeaponFiring"))))
    (:then_2
      ;;[0.03417 0.04445 0.15 0.5] Play melee attack montage.
      (Ability|Tasks|PlayMontageAndWaitNotify "MeleeAttackMontage" (Variables|Settings|GetMeleeMontage)
        (:OnCompleted
          _)
        (:OnCancelled
          _)) ; Play rate increased.
      (Default|WaitAsyncActionComplete (Variables|Transient|GetAsyncAction_DamageApplication)
        (:Complete
          (Ability|EndAbility))))))

(event Custom|HandleTargetingExecuted
  (params (TargetingHandle :type Targeting%20Request%20Handle%20Structure) (SwingIndex :type Integer))
  ;[0.09091 0.005675 0.09375 0.5] HandleTargetingExecuted: Callback after the melee targeting has been executed.
  (bind _outtargets (TargetingSystem|TargetingResults|GetTargetingResults (GameInstanceSubsystems|GetTargetingSubsystem) TargetingHandle))
  (Variables|Default|SetTemp%20Target%20Results _outtargets)
  (Utilities|FlowControl|DelayUntilNextTick) ; Can afford to wait a frame here for perf reasons. \nBy this point, target results have already been parsed once \nfor building actors in this frame.
  (Utilities|Array|Clear (Variables|Transient|GetHitResultsForMeleeDestructionVisuals))
  (for _array_element (Variables|Default|GetTemp%20Target%20Results)
    (bind _asbuilding_actor (Utilities|Casting|CastToBuildingActor (.HitActor (Collision|BreakHitResult _array_element)))
      (:then
        (if (GameplayTags|HasMatchingGameplayTag _asbuilding_actor "(TagName=\"TidalCrane.ToyFox.Boss\")")
          ;;[0 0.03485 0.15 0.5] Applying damage specifically to the enemy boss.
          (bind (_asynctaskproxy _targetdata) (Ability|Tasks|PerformTargetingAbilityTask "None" (Variables|TargetingPreset|GetTargetingPresetBoss) false false true)
            (:Targeted
              _)
            (:Cancelled
              _))
          (if (Ability|TargetData|TargetDataHasHitResult _targetdata)
            (Ability|ExecuteGameplayCueWithParamsOnOwner (Utilities|Struct|BreakGameplayCueTag (Variables|GameplayCue|GetBossMeleeDamageGameplayCueTag)) (Default|MakeGameplayCueParamfromHitResult (Ability|TargetData|GetHitResultfromTargetData _targetdata) SwingIndex))
            (Ability|ApplyGameplayEffectToTarget _targetdata (Variables|Settings|GetBossDamageGameplayEffectClass)))
          (else
            ;;[0.04973 0.15 0.08495 0.5] Applying damage to building actors that are not outright destroyed.
            (Ability|HasAuthority
              (:True
                (bind _giantpawndestructionmanager (Variables|Default|GetGiantPawnDestructionManager)
                  (:then
                    (bind _candestroy (Helper|IsActorValidforDestruction _giantpawndestructionmanager _asbuilding_actor))
                    (if _candestroy
                      (bind _asyncaction_damageapplication (Variables|Transient|GetAsyncAction_DamageApplication)
                        (:then
                          (if (AsyncAction|IsActive _asyncaction_damageapplication)
                            (Ability|Tasks|AddTargetData _asyncaction_damageapplication (Ability|TargetData|AbilityTargetDatafromHitResult _array_element))
                            (else
                              (bind _asynctaskproxy_2 (Ability|Tasks|ApplyGameplayEffectToTargetDataAsync (Ability|TargetData|AbilityTargetDatafromHitResult _array_element) (Variables|Settings|GetDamageGameplayEffectClass) 1 1 (Variables|Hotfix|GetHotfix_MaxTargetsPerTickToDamage)))
                              (Variables|Transient|SetAsyncAction_DamageApplication _asynctaskproxy_2)))
                          (Utilities|Array|Add (Variables|Transient|GetHitResultsForMeleeDestructionVisuals) _array_element))
                        (:else
                          _))
                      (else
                        _)))
                  (:else)))
              (:False
                _)))))
      (:CastFailed
        (Utilities|FlowControl|Sequence
          (:then_0
            _)
          (:then_1
            ;;[0.06736 0.03278 0.15 0.5] Melee attack explodes projectiles it impacts.
            (bind _asfort_projectile_base (Utilities|Casting|CastToFortProjectileBase (.HitActor (Collision|BreakHitResult _array_element))))
            (Delivery|DoExplosion _asfort_projectile_base))))))
  (Utilities|FlowControl|DelayUntilNextTick) ; Can afford to wait a frame here for perf reasons. 
  (Default|ExecuteMeleeDestructionGameplayCue (Variables|Transient|GetHitResultsForMeleeDestructionVisuals) SwingIndex))

(event Ability|EventOnEndAbility
  (params (bWasCancelled :type Boolean))
  ;[0.08854 0.05454 0 0.5] OnEndAbility: Unbind delegates.
  (Utilities|FlowControl|Sequence
    (:then_0
      (Ability|RemoveGameplayEffectFromOwnerWithHandle (Variables|Transient|GetMeleeSwing1EffectHandle))
      (Ability|RemoveGameplayEffectFromOwnerWithHandle (Variables|Transient|GetMeleeSwing2EffectHandle))
      (Ability|RemoveGameplayEffectFromOwnerWithHandle (Variables|Transient|GetMeleeActiveGameplayEffectHandle))
      (bind _weapondestructioncomponent (Variables|Transient|GetWeaponDestructionComponent)
        (:then
          (Default|UnbindEventfromOnTargetingExecuted _weapondestructioncomponent (EventDispatchers|CreateEvent :!event Custom|HandleTargetingExecuted)))
        (:else)))
    (:then_1
      (|CallParentFunction bWasCancelled))))

(event Ability|EventOnGenericNotifyFromAnimation
  (params (NotifyTag :type Gameplay%20Tag%20Structure))
  ;[0.0148 0.15 0.001496 0.5] OnGenericNotifyFromAnimation: Notify from the animation. Apply gameplay effect to track the melee swing done by the pawn.
  (switch Utilities|FlowControl|Switch|SwitchonGameplayTag NotifyTag
    (:Gameplay.Melee.Combo1
      (bind _returnvalue (Ability|ApplyGameplayEffectToOwner (Variables|GameplayEffects|GetMeleeSwing1GameplayEffectClass)))
      (Variables|Transient|SetMeleeSwing1EffectHandle _returnvalue))
    (:Gameplay.Melee.Combo2
      (bind _returnvalue_2 (Ability|ApplyGameplayEffectToOwner (Variables|GameplayEffects|GetMeleeSwing2GameplayEffectClass)))
      (Variables|Transient|SetMeleeSwing2EffectHandle _returnvalue_2))))

(fn ExecuteMeleeDestructionGameplayCue
  (params (HitResults :type Array%20of%20Hit%20Result%20Structures) (SwingIndex :type Integer))
  (bind _returnvalue (Networking|IsDedicatedServer))
  (bind _returnvalue_2 (select _returnvalue "(R=0.000000,G=1.000000,B=0.021790,A=1.000000)" "(R=0.000000,G=1.000000,B=0.933321,A=1.000000)"))
  (bind (_minvector_x _minvector_y _minvector_z) (Math|Vector|BreakVector (Variables|GetMinVector_X)))
  (bind (_maxvector_x _maxvector_y _maxvector_z) (Math|Vector|BreakVector (Variables|GetMaxVector_X)))
  (bind _maxvector (LocalVariables|GetMaxVector))
  ;[0 0.15 0.02033 0.5] ExecuteMeleeDestructionGameplayCue: Calculate the center of mass and box extent of all the building actors that were hit. Pass that to the gameplay cue so it can play the FX accordingly.
  (if (> (Utilities|Array|Length HitResults) 0)
    (for _array_element HitResults
      (Utilities|IsValid (.HitActor (Collision|BreakHitResult _array_element))
        (:Is%20Valid
          (Utilities|IsValid (.HitComponent (Collision|BreakHitResult _array_element))
            (:Is%20Valid
              (bind _origin (Collision|GetComponentBounds (.HitComponent (Collision|BreakHitResult _array_element))))
              (bind (_origin_2 _boxextent) (Collision|GetActorBounds (.HitActor (Collision|BreakHitResult _array_element)) false false))
              (bind _output_get (LocalVariables|SetActorMin (- _origin_2 _boxextent)))
              (bind _output_get_2 (LocalVariables|SetActorMax (+ _origin_2 _boxextent)))
              (Variables|SetMinVector_X (Math|Vector|MakeVector (Math|Float|Min(Float) _minvector_x (.x _output_get)) (Math|Float|Min(Float) _minvector_y (.y _output_get)) (Math|Float|Min(Float) _minvector_z (.z _output_get))))
              (Variables|SetMaxVector_X (Math|Vector|MakeVector (Math|Float|Max(Float) _maxvector_x (.x _output_get_2)) (Math|Float|Max(Float) _maxvector_y (.y _output_get_2)) (Math|Float|Max(Float) _maxvector_z (.z _output_get_2))))
              (bind _output_get_3 (LocalVariables|SetVolume (Utilities|Operators|Multiply (* (.x _boxextent) 2.0) (* (.y _boxextent) 2.0) (* (.z _boxextent) 2.0))))
              (LocalVariables|SetWeightedLocationSum (+ (LocalVariables|GetWeightedLocationSum) (* _origin _output_get_3)))
              (LocalVariables|SetTotalWeight (+ _output_get_3 (LocalVariables|GetTotalWeight))))))))
    (bind _output_get_4 (LocalVariables|SetCenterOfMass (/ (LocalVariables|GetWeightedLocationSum) (LocalVariables|GetTotalWeight))))
    (LocalVariables|SetBoxExtentCenter (/ (+ (LocalVariables|GetMinVector) _maxvector) 2.0))
    (LocalVariables|SetBoxExtent (/ (- _maxvector (LocalVariables|GetMinVector)) 2.0))
    (bind _output_get_5 (LocalVariables|SetGameplayCueEffectContext (GameplayEffects|MakeEffectContext (Ability|GetAbilitySystemComponentfromActorInfo))))
    (Ability|EffectContext|SetOrigin _output_get_5 (LocalVariables|GetBoxExtent))
    (Ability|ExecuteGameplayCueWithParamsOnOwner (Utilities|Struct|BreakGameplayCueTag (Variables|GameplayCue|GetBuildingDestructionGameplayCueTag)) (Ability|GameplayCue|MakeGameplayCueParameters :NormalizedMagnitude 0.0 :RawMagnitude (Math|Conversions|ToFloat(Integer) SwingIndex) :EffectContext (LocalVariables|GetGameplayCueEffectContext) :Location (LocalVariables|GetCenterOfMass) :Normal "0,0,0" :GameplayEffectLevel 1 :AbilityLevel 1 :bReplicateLocationWhenUsingMinimalRepProxy false))
    (if (Variables|Debug|GetbDebug)
      (Rendering|Debug|DrawDebugSphere _output_get_4 1e+02 12 _returnvalue_2 8.0 5.0)
      (Rendering|Debug|DrawDebugBox :Center (LocalVariables|GetCenterOfMass) :Extent (LocalVariables|GetBoxExtentCenter) :LineColor _returnvalue_2 :Duration 8.0 :Thickness 5.0 :DepthPriority "World"))))

(fn MakeGameplayCueParamFromHitResult
  (params (HitResult :type Hit%20Result%20Structure%20%28by%20ref%29) (SwingIndex :type Integer))
  (bind (_normalizedmagnitude _rawmagnitude _effectcontext _matchedtagname _originaltag _aggregatedsourcetags _aggregatedtargettags _location _normal _instigator _effectcauser _sourceobject _physicalmaterial _gameplayeffectlevel _abilitylevel _targetattachcomponent _breplicatelocationwhenusingminimalrepproxy) (Ability|GameplayCue|BreakGameplayCueParameters (GameplayCue|MakeGameplayCueParametersfromHitResult HitResult)))
  (bind _output_get (LocalVariables|SetEffectContext (GameplayEffects|MakeEffectContext (Ability|GetAbilitySystemComponentfromActorInfo))))
  (Ability|EffectContext|AddHitResult _output_get HitResult)
  (return (Ability|GameplayCue|MakeGameplayCueParameters _normalizedmagnitude (Math|Conversions|ToFloat(Integer) SwingIndex) _output_get _matchedtagname _originaltag _aggregatedsourcetags _aggregatedtargettags _location _normal _instigator _effectcauser _sourceobject _physicalmaterial _gameplayeffectlevel _abilitylevel _targetattachcomponent _breplicatelocationwhenusingminimalrepproxy)))
